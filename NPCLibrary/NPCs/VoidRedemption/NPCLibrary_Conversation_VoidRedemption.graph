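# Conversation database for the VoidRedemption NPC, defining its greeting and top-level conversation branches.
# The entry point feeds one Speech node that chooses contextual lines by requirement and priority: first-ever meeting, player outfit reactions, or general lines gated by match conversation count.
# That node fans out through link nodes, by remote entry tag, to basic services, NPC library quests, character dialogue, a Ch5S1 conversation and Season 30 story conversations.
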ConversationDatabase /NPCLibrary/NPCs/VoidRedemption/NPCLibrary_Conversation_VoidRedemption
entry_points [{tag=Conversation.Entry.VoidRedemption, destinations [#1]}]
nodes [
 ConversationEntryPointNode_7 {
  id=#1
  class=ConversationEntryPointNode
  outputs [#2]
  properties {entryTag {tagName=Conversation.Entry.VoidRedemption}}
 }
 Speech_C_4 {
  id=#2
  class=Speech_C
  outputs [
   #3
   #4
   #5
   #6
   #7
   #8
   #9
   #10
   #11
   #12
  ]
  properties {
   nodeName=""
   generalConfig {
    defaultMessage=I'm gonna keep knocking down these villains until they stop getting up.
    contextualMessages [
     {
      message="Hey. Here we go again, right?"
      contextRequirements [{requirement {class=FortControllerRequirement_IsFirstEverConversationWithNPC}}]
      priority {curve {rowName=Tandem.Speech.Priority.FirstEverInteraction}}
     }
     {
      message="Sometimes I feel like if I stop being angry, I'll just sit down and stop fighting entirely."
      contextRequirements [{requirement {
         class=FortControllerRequirement_HasCID
         properties {allowedCIds [&Character_UndergroundRebel_Fashion]}
        }}]
      priority {curve {rowName=Tandem.Speech.Priority.OutfitReaction}}
     }
     {
      message=I'm still mad at you... but I also wish you'd just come home.
      contextRequirements [{requirement {class=FortControllerRequirement_HasCID, properties {allowedCIds [/BRCosmetics/Athena/Items/Cosmetics/Characters/Character_MadameMoth_Posh]}}}]
      priority {curve {rowName=Tandem.Speech.Priority.OutfitReaction}}
     }
     {
      message=Now that everything's changed... don't you want to be part of something good?
      contextRequirements [{requirement {
         class=FortControllerRequirement_HasCID
         properties {allowedCIds [/BRCosmetics/Athena/Items/Cosmetics/Characters/Character_CeremonialGuard_Fencer]}
        }}]
      priority {curve {rowName=Tandem.Speech.Priority.OutfitReaction}}
     }
     {
      message=This wasteland is... a lot. Thanks for having my back.
      contextRequirements [{requirement {
         class=FortControllerRequirement_HasCID
         properties {allowedCIds [
           /BRCosmetics/Athena/Items/Cosmetics/Characters/Character_ShiverFlame_Ace
           /BRCosmetics/Athena/Items/Cosmetics/Characters/Character_QuickBurst_Plains
          ]}
        }}]
      priority {curve {rowName=Tandem.Speech.Priority.OutfitReaction}}
     }
     {
      message="I guess I was wrong. You're always going to put yourself first, aren't you?"
      contextRequirements [{requirement {
         class=FortControllerRequirement_HasCID
         properties {allowedCIds [&Character_BengalBasher_Suit]}
        }}]
      priority {curve {rowName=Tandem.Speech.Priority.OutfitReaction}}
     }
     {
      message=I'm gonna keep knocking down these villains until they stop getting up.
      contextRequirements [{requirement {
         class=FortControllerRequirement_CheckNPCMatchConversationCount
         properties {nPCUniqueIdTag {tagName=AISpawnerData.Type.Tandem.VoidRedemption}}
        }}]
      priority {curve {rowName=None}}
     }
     {
      message=I have this sneaking suspicion that there's more going on than we know...
      contextRequirements [{requirement {
         class=FortControllerRequirement_CheckNPCMatchConversationCount
         properties {nPCUniqueIdTag {tagName=AISpawnerData.Type.Tandem.VoidRedemption}}
        }}]
      priority {curve {rowName=None}}
     }
     {
      message=Things used to be so simple. But this is better.
      contextRequirements [{requirement {
         class=FortControllerRequirement_CheckNPCMatchConversationCount
         properties {nPCUniqueIdTag {tagName=AISpawnerData.Type.Tandem.VoidRedemption}}
        }}]
      priority {curve {rowName=None}}
     }
    ]
   }
  }
 }
 ConversationLinkNode_9 {
  id=#12
  class=ConversationLinkNode
  properties {remoteEntryTag {tagName=Conversation.Entry.Season30.Story.P06.Q05.VR}}
 }
 ConversationLinkNode_8 {
  id=#11
  class=ConversationLinkNode
  properties {remoteEntryTag {tagName=Conversation.Entry.Season30.Story.P05.Q06.VR}}
 }
 ConversationLinkNode_7 {
  id=#10
  class=ConversationLinkNode
  properties {remoteEntryTag {tagName=Conversation.Entry.Season30.Story.P04.Q07.VoidRedemption}}
 }
 ConversationLinkNode_6 {
  id=#9
  class=ConversationLinkNode
  properties {remoteEntryTag {tagName=Conversation.Entry.Season30.Story.P03.Q06.VoidRedemption}}
 }
 ConversationLinkNode_5 {
  id=#8
  class=ConversationLinkNode
  properties {remoteEntryTag {tagName=Conversation.Entry.Season30.Story.P02.06.VoidRedemption}}
 }
 ConversationLinkNode_4 {
  id=#7
  class=ConversationLinkNode
  properties {remoteEntryTag {tagName=Conversation.Entry.Season30.Story.P01.Q01.VR}}
 }
 ConversationLinkNode_3 {
  id=#6
  class=ConversationLinkNode
  properties {remoteEntryTag {tagName=Conversation.Entry.VoidRedemption.Dialogue}}
 }
 ConversationLinkNode_2 {id=#5, class=ConversationLinkNode, properties {remoteEntryTag {tagName=Conversation.Entry.Ch5S1.D3.Q8}}}
 ConversationLinkNode_0 {id=#4, class=ConversationLinkNode, properties {remoteEntryTag {tagName=Conversation.Entry.BasicServices}}}
 ConversationLinkNode_1 {id=#3, class=ConversationLinkNode, properties {remoteEntryTag {tagName=Conversation.Entry.NPCLibrary.Quests}}}
]
