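# Conversation database for the VoidRedemption NPC's "Ask about..." dialogue, where the NPC shares thoughts on recent events.
# The entry tag leads to an intro line that branches into The Gods, The Oracle, Valeria (the NPC's sister) and a Back option.
# Each topic plays a linear three-line Speech chain, and all three converge on a shared ReturnToStart node labeled BACK.
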ConversationDatabase /NPCLibrary/NPCs/VoidRedemption/NPCLibrary_Conversation_VoidRedemption_Dialogue
entry_points [{tag=Conversation.Entry.VoidRedemption.Dialogue, destinations [#1]}]
nodes [
 ConversationEntryPointNode_0 {
  id=#1
  class=ConversationEntryPointNode
  outputs [#2]
  properties {entryTag {tagName=Conversation.Entry.VoidRedemption.Dialogue}}
 }
 Speech_C_0 {
  id=#2
  class=Speech_C
  outputs [#3, #4, #5, #6]
  properties {
   nodeName=""
   softTaskIcon=&T_UI_NPC_question
   generalConfig {defaultMessage="If I'm honest, my head's still spinning. But I'll tell you what I know."}
  }
  sub_nodes [ConversationChoiceNode_5 {class=ConversationChoiceNode, properties {defaultChoiceDisplayText=Ask about...}}]
 }
 Speech_C_1 {
  id=#6
  class=Speech_C
  outputs [#7]
  properties {
   nodeName=""
   softTaskIcon=&T_UI_NPC_question
   generalConfig {defaultMessage=There's no way my sister could have known that all of this was going to happen...}
  }
  sub_nodes [ConversationChoiceNode_10 {class=ConversationChoiceNode, properties {defaultChoiceDisplayText=Valeria}}]
 }
 Speech_C_2 {
  id=#7
  class=Speech_C
  outputs [#8]
  properties {
   nodeName=""
   softTaskIcon=&T_UI_NPC_question
   generalConfig {defaultMessage="... But at the same time, she did this. She brought this upon us."}
  }
 }
 Speech_C_3 {
  id=#8
  class=Speech_C
  outputs [#9]
  properties {
   nodeName=""
   softTaskIcon=&T_UI_NPC_question
   generalConfig {defaultMessage="I love her. I'm mad at her. She'd BETTER be sorry... (sigh) And I miss her."}
  }
 }
 ReturnToStart_C_23 {
  id=#9
  class=ReturnToStart_C
  properties {softTaskIcon=&T_UI_NPC_Back}
  sub_nodes [ConversationChoiceNode_32 {class=ConversationChoiceNode, properties {defaultChoiceDisplayText=BACK}}]
 }
 Back_C_0 {
  id=#5
  class=Back_C
  sub_nodes [ConversationChoiceNode_7 {class=ConversationChoiceNode, properties {defaultChoiceDisplayText=BACK}}]
 }
 Speech_C_6 {
  id=#4
  class=Speech_C
  outputs [#10]
  properties {nodeName="", softTaskIcon=&T_UI_NPC_question, generalConfig {defaultMessage="...Okay, I was wrong about her."}}
  sub_nodes [ConversationChoiceNode_0 {class=ConversationChoiceNode, properties {defaultChoiceDisplayText=The Oracle}}]
 }
 Speech_C_12 {
  id=#10
  class=Speech_C
  outputs [#11]
  properties {nodeName="", softTaskIcon=&T_UI_NPC_question, generalConfig {defaultMessage=The Oracle is real.}}
 }
 Speech_C_13 {
  id=#11
  class=Speech_C
  outputs [#9]
  properties {
   nodeName=""
   softTaskIcon=&T_UI_NPC_question
   generalConfig {defaultMessage="And frankly, she's kind of rude. I like her."}
  }
 }
 Speech_C_7 {
  id=#3
  class=Speech_C
  outputs [#12]
  properties {
   nodeName=""
   softTaskIcon=&T_UI_NPC_question
   generalConfig {defaultMessage=I used to think the Society was the worst thing there was.}
  }
  sub_nodes [ConversationChoiceNode_15 {class=ConversationChoiceNode, properties {defaultChoiceDisplayText=The Gods}}]
 }
 Speech_C_15 {
  id=#12
  class=Speech_C
  outputs [#13]
  properties {
   nodeName=""
   softTaskIcon=&T_UI_NPC_question
   generalConfig {defaultMessage="Getting attacked by angry gods will change your perspective pretty quickly, though."}
  }
 }
 Speech_C_14 {
  id=#13
  class=Speech_C
  outputs [#9]
  properties {
   nodeName=""
   softTaskIcon=&T_UI_NPC_question
   generalConfig {defaultMessage="Don't get me wrong, they're both bad. But I really wasn't a fan of getting electrocuted."}
  }
 }
]
